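# Translucent UI material for adaptive text on UIKit CTA buttons, deriving its color from the screen content behind it.
# It samples Slate post buffer 0, runs that through the CoreUI BackplateD function, and blends raw and processed colors with the Color Blend parameter.
# The blend feeds the Idle, Hovered, Focused and Pressed inputs of the shared ModularButton state-lerp function, leaving Disabled unconnected.
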Material /FortUILibrary/UIKit/Buttons/Materials/CTA/M_UI_CTA_AdaptiveText
graph {
 properties {
  bAutomaticallySetUsageInEditor=false
  bUsedWithStaticMesh=false
  blendMode=BLEND_Translucent
  materialDomain=MD_UI
  pixelDepthOffsetMode=PDOM_Legacy
  usedShadingModels=Default Lit
 }
 nodes [
  n1 = MaterialFunctionCall!("UVs (Optional)") {materialFunction=/Engine/Functions/UserInterface/GetSlatePost0}
  n2 = LinearInterpolate(
   A=@n1.LinearRGB
   B=MaterialFunctionCall!(Post Buffer Blur=@n1.LinearRGB) {materialFunction=/FortUILibrary/MaterialFunctions/CoreUI/MF_UI_CoreUI_BackplateD}.Result
   Alpha=ScalarParameter() {defaultValue=1, parameterName=Color Blend, sliderMax=1}
  )
  MaterialFunctionCall!(Idle=@n2, Idle Opacity, Hovered=@n2, Hovered Opacity, Focused=@n2, Focused Opacity, Pressed=@n2, Pressed Opacity, Disabled, Disabled Opacity) {materialFunction=/FortUILibrary/Systems/ModularButton/MaterialFunctions/MF_UIKit_ButtonStatesBase_Lerp}
 ]
}
